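; Parent actor for projectile fuel in the Curie fire system, which lands on surfaces and can ignite itself and nearby materials.
; On hit with Curie active, it attaches to vehicles, doors or buildings, optionally starts fires through a sphere overlap or on itself, and stops server-side movement on non-pawn, non-fuel hits.
; It silent-dies when its base dies, and expires quickly when doused or hit while Curie is inactive.
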(event EventBeginPlay
  ;[1 1 1] Set Curie component when Curie's activated
  (Curie|OnCurieActivated
    (:OnCurieActive
      (Variables|Default|SetCurieActive true))))

(event Collision|EventHit
  (params (MyComp :type Primitive%20Component%20Object%20Reference) (Other :type Actor%20Object%20Reference) (OtherComp :type Primitive%20Component%20Object%20Reference) (bSelfMoved :type Boolean) (HitLocation :type Vector :default "0, 0, 0") (HitNormal :type Vector :default "0, 0, 0") (NormalImpulse :type Vector :default "0, 0, 0") (Hit :type Hit%20Result%20Structure%20%28by%20ref%29))
  (bind _hitactor (Variables|Default|GetHitActor))
  ;[1 1 1] On Hit, run through logic to see what type of actor we hit, attach to actor where possible and then stop simulating the projectilemovement
  (if (not (Variables|Default|GetIgnoreHits))
    (Variables|Default|SetHitActor Other)
    (Variables|Default|SetHitComponent OtherComp)
    (if (Variables|Default|GetCurieActive)
      (CallFunction|HandleHits (Variables|Default|GetHitActor) (Variables|Default|GetHitComponent))
      (if (Variables|CurieFuel|GetIgniteNearbyMaterialsImmediately)
        (CallFunction|IgniteNearbyMaterials HitLocation)
        (else
          _))
      (if (Variables|CurieFuel|GetIgniteSelf)
        (CallFunction|SelfIgnite)
        (else
          _))
      (Utilities|Casting|CastToPawn _hitactor
        (:CastFailed
          (Utilities|Casting|CastToBGA_GenericCurieFuel_Parent _hitactor
            (:CastFailed
              (if (Networking|IsServer)
                (Networking|FlushNetDormancy)
                (Game|Components|ProjectileMovement|StopSimulating (Variables|Default|GetFortProjectileMovement) Hit)
                ;[1 1 1] Bind event to on death that will either cause the mesh to silent die or fall with gravity
                (Utilities|IsValid _hitactor
                  (:Is%20Valid
                    (bind _asbuilding_actor (Utilities|Casting|CastToBuildingActor _hitactor)
                      (:then
                        (EventDispatchers|BindEventtoOnDied _asbuilding_actor (EventDispatchers|CreateEvent :!event Custom|AttachedComponentDied)))
                      (:CastFailed
                        (Game|BindEventtoOnDestroyed _hitactor (EventDispatchers|CreateEvent :!event Custom|BaseDied)))))))))))
      (else
        (Actor|SetLifeSpan 0.1)))))

(event OnProjectileBounce(FortProjectileMovement)
  (params (ImpactResult :type Hit%20Result%20Structure%20%28by%20ref%29) (ImpactVelocity :type Vector%20%28by%20ref%29 :default "0, 0, 0"))
  ;[1 1 1] @Mike - comment would be good here.  Seems like a vehicle thing?
  (Networking|FlushNetDormancy)
  (Variables|/Script/Engine.ProjectileMovementComponent|SetbShouldBounce false (Variables|Default|GetFortProjectileMovement))
  (CallFunction|HandleHits (Variables|Default|GetHitActor) (Variables|Default|GetHitComponent)))

(event Curie|EventOnCurieElementDetached
  (params (CurieContainerHandle :type Curie%20Container%20Handle%20Structure) (ElementTag :type Gameplay%20Tag%20Structure))
  ;[1 1 1] Kill the fuel if it's doused.  Set life span rather than destroy to stagger cleanup.
  (if (GameplayTags|Equal(GameplayTag) ElementTag "(TagName=\"Curie.Element.Fire\")")
    (Rendering|SetActorHiddenInGame true)
    (Actor|SetLifeSpan 0.2)))

(event Custom|AttachedComponentDied
  (params (DamagedActor :type Actor%20Object%20Reference) (Damage :type Float%20%28single-precision%29) (InstigatedBy :type Controller%20Object%20Reference) (DamageCauser :type Actor%20Object%20Reference) (HitLocation :type Vector) (FHitComponent :type Primitive%20Component%20Object%20Reference) (BoneName :type Name) (Momentum :type Vector))
  ;[1 1 1] On Death, based on bool we either fall with gravity or we silent die
  (Gameplay|SilentDie false))

(event OnProjectileStop(FortProjectileMovement)
  (params (ImpactResult :type Hit%20Result%20Structure%20%28by%20ref%29))
  (Variables|Default|SetIgnoreHits true)
  (Collision|SetCollisionResponsetoChannel (Variables|Default|GetSphere) "ECC_Vehicle" "ECR_Ignore") ; Disable vehicle collision response to fix mech flying
  (Networking|ForceNetUpdate))

(event Custom|BaseDied
  (params (DestroyedActor :type Actor%20Object%20Reference))
  ;[1 1 1] On Death, based on bool we either fall with gravity or we silent die
  (Gameplay|SilentDie false))

(fn HandleHits
  (params (ObjectRef :type Object%20Reference) (PrimitiveObjectRef :type Primitive%20Component%20Object%20Reference))
  (Variables|Default|SetCurieHitVehicle false)
  (Utilities|FlowControl|Sequence
    (:then_0
      ;[0.9211 0 1] Pulling this out until we have a solution formoving fires
      (bind _success (CallFunction|HitVehicle ObjectRef PrimitiveObjectRef))
      (if _success
        _))
    (:then_1
      (bind _success_2 (CallFunction|HitBuildingWall ObjectRef PrimitiveObjectRef))
      (if _success_2
        _))
    (:then_2
      (bind _success_3 (CallFunction|HitBuildingActor ObjectRef PrimitiveObjectRef))
      (if _success_3
        _)))
  (return))

(fn HitBuildingActor
  (params (ObjectRef :type Object%20Reference) (PrimitiveObjectRef :type Primitive%20Component%20Object%20Reference))
  (bind _asbuilding_actor (Utilities|Casting|CastToBuildingActor ObjectRef))
  (Utilities|Casting|CastToBGA_GenericCurieFuel_Parent _asbuilding_actor
    (:CastFailed
      (Transformation|AttachActorToActor _asbuilding_actor "None" "KeepWorld" "KeepWorld" "KeepWorld" false)
      (return true))))

(fn HitBuildingWall
  (params (ObjectRef :type Object%20Reference) (PrimitiveObjectRef :type Primitive%20Component%20Object%20Reference))
  (bind _asbuilding_wall (Utilities|Casting|CastToBuildingWall ObjectRef))
  (Variables|Default|SetLocalWall _asbuilding_wall)
  (bind _asstatic_mesh_component (Utilities|Casting|CastToStaticMeshComponent PrimitiveObjectRef))
  (bind _returnvalue (BuildingWall|IsDoorComponent (Variables|Default|GetLocalWall) _asstatic_mesh_component))
  (if _returnvalue
    (Transformation|AttachActorToComponent PrimitiveObjectRef "None" "KeepWorld" "KeepWorld" "KeepWorld" false)
    (return true)))

(fn HitVehicle
  (params (ObjectRef :type Object%20Reference) (PrimitiveObjectRef :type Primitive%20Component%20Object%20Reference))
  (Utilities|Casting|CastToFortVehicleInterface ObjectRef)
  (Transformation|AttachActorToComponent PrimitiveObjectRef "None" "KeepWorld" "KeepWorld" "KeepWorld" false)
  (Collision|SetCollisionResponsetoChannel (Variables|Default|GetSphere) "ECC_Vehicle" "ECR_Ignore") ; Disable vehicle collision response to fix mech flying
  (return true))

(fn IgniteNearbyMaterials
  (params (SpherePos :type Vector))
  ;[1 1 1] If ImmediatelyAttemptIgnition is TRUE, we will do a sphereoverlap at the location of the hit and immediately set fire to all valid actors.
  (bind _outactors (Collision|SphereOverlapActors SpherePos (Variables|CurieFuel|GetIgniteNearbyMaterialsRadius) (Variables|CurieFuel|GetIgniteMaterialsObjectTypes) (Variables|CurieFuel|GetIgniteMaterialsClassFilter) (Utilities|Array|MakeArray self)))
  (for _array_element _outactors
    (Utilities|IsValid _array_element
      (:Is%20Valid
        (Curie|ExecuteFireApplication (Fortnite|GetFortPlayerControllerfromActor (Actor|GetOwner)) self _array_element (Variables|CurieFuel|GetIgniteNearbyMaterialsFireMagnitude) (Collision|MakeHitResult false false 0.0 0.0 SpherePos SpherePos "0,0,1" "0,0,1" 0 _array_element) (Variables|CurieFuel|GetIgniteNearbyMaterialsPropogationFuel) (Variables|CurieFuel|GetIgniteNearbyMaterialsLandscapeIgnitionRadius))))))

(fn OnRep_ResumeSimulation
  (if (Variables|CurieFuel|GetResumeSimulation)
    (Components|Movement|SetUpdatedComponent (Variables|Default|GetFortProjectileMovement) (Variables|Default|GetSphere))))

(fn SelfIgnite
  (Networking|SwitchHasAuthority
    (:Authority
      ;[1 1 1] Set how long the mesh exists for in the world after catching fire
      (if (not (Variables|Default|GetCurieHitVehicle))
        (Actor|SetLifeSpan (Variables|CurieFuel|GetActorLifespanAfterIgnition))
        (Curie|ExecuteFireApplication :Instigator (Fortnite|GetFortPlayerControllerfromActor (Actor|GetOwner)) :EffectCauser self :Target self :Magnitude (Variables|CurieFuel|GetSelfFireMagnitude) :PropagationFuel (Variables|CurieFuel|GetSelfPropagationFuel) :LandscapeIgnitionRadius (Variables|CurieFuel|GetSelfCurieLandscapeIgnitionRadius))))))

(fn ShouldPlayNativeCurieFX
  (params (FXTag :type Gameplay%20Tag%20Structure))
  (return (Utilities|Struct|MakeFortNativeCurieFXResponse :GameplayCueResponse "IgnoreCue" :bShouldPlayGeneralVFX true :bShouldPlayAmbientAudio true :bShouldPlayGlow true :bShouldPlayFXAsAOE true)))

(fn ConstructionScript
  (|CallParentFunction))
