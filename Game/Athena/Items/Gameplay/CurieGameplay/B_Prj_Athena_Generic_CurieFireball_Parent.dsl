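; Parent fireball projectile that leaves a burning patch where it lands by spawning a configurable Curie building gameplay actor.
; On hit it stops movement, optionally rolls a weighted chance to skip the patch, and line traces for an optional floor-aligned spawn transform.
; A short timer then spawns the actor and switches off the fire trail.
; Entering water plays a one-time fizzle sound and kills the trail.
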(event EventBeginPlay
  ;[1 1 1] Set the Particle to what we desire
  (Effects|Components|ParticleSystem|SetTemplate (Variables|Default|GetFireTrail) (Variables|CurieProjectile|GetDesiredFireTrailParticle)))

(event Custom|DelaySpawnCurieBGA
  (CallFunction|SpawnCurieBGA))

(event Custom|Landed%20In%20Water
  ;[1 1 1] This attempts to adjust the angle of the mesh based on the floor angle.
  (if (Variables|Default|GetShouldSpawnFire)
    (Utilities|Time|SetTimerbyEvent (delegate Custom|DelaySpawnCurieBGA) 0.1)))

(event Custom|SpawnCurieBGA
  ;[1 1 1] Spawn the curie BGA
  (Building|SpawnBuildingGameplayActor (Variables|CurieProjectile|GetBGAToSpawn) (Variables|Default|GetLineTraceTransform) (Actor|GetOwner))
  ;[1 1 1] Turn off the particle
  (Components|Activation|Deactivate (Variables|Default|GetFireTrail)))

(event Collision|EventHit
  (params (MyComp :type Primitive%20Component%20Object%20Reference) (Other :type Actor%20Object%20Reference) (OtherComp :type Primitive%20Component%20Object%20Reference) (bSelfMoved :type Boolean) (HitLocation :type Vector :default "0, 0, 0") (HitNormal :type Vector :default "0, 0, 0") (NormalImpulse :type Vector :default "0, 0, 0") (Hit :type Hit%20Result%20Structure%20%28by%20ref%29))
  (Game|Components|ProjectileMovement|StopSimulating (Variables|Movement|GetProjectileMovementComponent) Hit)
  ;[1 1 1] Do we want each projectile to have a random chance to not create a burning patch. If yes, set ShouldFireSpawnWithRandomChance to True and define the Weight
  (if (Variables|CurieProjectile|GetShouldFireSpawnWithRandomChance)
    (if (not (Math|Random|RandomBoolwithWeight (Variables|CurieProjectile|GetRandomSpawnChanceWeight)))
      ;[1 1 1] Turn off the particle
      (Components|Activation|Deactivate (Variables|Default|GetFireTrail)))
    (else
      _))
  ;[1 1 1] This attempts to adjust the angle of the mesh based on the floor angle.
  (bind _returnvalue (CallFunction|LineTraceForMeshAngle))
  (Variables|Default|SetLineTraceTransform _returnvalue)
  (if (Variables|Default|GetShouldSpawnFire)
    (Utilities|Time|SetTimerbyEvent (delegate Custom|DelaySpawnCurieBGA) 0.1)))

(event OnEnterWaterEvent(WaterInteractionComponent)
  (params (WaterBody :type Fort%20Water%20Body%20Actor%20Object%20Reference) (WaterInteractionComponent :type Fort%20Water%20Interaction%20Component%20Object%20Reference) (bIsFirstBody :type Boolean))
  ;[1 1 1] On water impact fizzle
  (Utilities|FlowControl|DoOnce false
    (:Completed
      (Audio|PlaySoundatLocation (Variables|Default|GetFireFizzle) (Transformation|GetActorLocation))
      (Components|Activation|Deactivate (Variables|Default|GetFireTrail)))))

(fn LineTraceforMeshAngle
  (bind _capsulecomponent (Variables|Collision|GetCapsuleComponent))
  (bind _returnvalue (Transformation|GetWorldLocation _capsulecomponent))
  (bind _outhit (Collision|LineTraceByChannel _returnvalue (+ _returnvalue (Variables|CurieProjectile|GetFire%20Plane%20Trace%20Offset)) "TraceTypeQuery10"))
  (return (Math|Transform|SelectTransform (Math|Transform|SelectTransform (Math|Transform|MakeTransform (+ (.Location (Collision|BreakHitResult _outhit)) (* (.Normal (Collision|BreakHitResult _outhit)) (Variables|CurieProjectile|GetFire%20Plane%20Spawn%20Offset))) (Math|Rotator|MakeRotfromZX (.Normal (Collision|BreakHitResult _outhit)) (1.0 0.0 0.0))) (Math|Transform|MakeTransform _returnvalue) (Collision|BreakHitResult _outhit)) (Math|Transform|MakeTransform _returnvalue) (Variables|CurieProjectile|GetDoesMeshAngleNeedAdjusted))))

(fn ConstructionScript)
